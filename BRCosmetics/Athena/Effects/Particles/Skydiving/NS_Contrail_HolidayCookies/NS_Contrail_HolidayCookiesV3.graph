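# Niagara skydiving contrail with a holiday cookie theme, spawned from the user-supplied skeletal mesh.
# Red and green ribbons and color-cycling sprites with lights follow the mesh, with time-of-day alpha scaling, alongside twinkling star sprites.
# Tumbling cookie man and tree meshes scale spawn rate and velocity with owner speed, drift on curl noise, and generate location events.
# Separate fragment emitters render small, short-scaled mesh bits for each cookie type.
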NiagaraSystem /BRCosmetics/Athena/Effects/Particles/Skydiving/NS_Contrail_HolidayCookies/NS_Contrail_HolidayCookiesV3
dependencies {
 usedDataInterfaces [&NiagaraDataInterfaceCurve, &NiagaraDataInterfaceColorCurve]
 usedDynamicInputs []
 usedModules [&SystemState]
}
emitters {
 Contrail_Green {
  emitterUpdateScript [
   &EmitterState()
   &FN_SpawnRibbon_v2(Number of Ribbons=2, Spawn Mode=::NewEnumerator0, SpawnRate=5)
  ]
  particleSpawnScript [
   /Game/Effects/Niagara/Modules/Ribbons/FN_InitializeRibbon_v5(Enable Skeletal Position=true, Lifetime=1, Ribbon Width=28, SkeletalMesh:NiagaraDataInterfaceSkeletalMesh=$User.SkeletalMesh)
   &AddVelocity(Velocity=&FN_FrontEndVector3Modifier(Gameplay=<0 0 500>, Front End=<0 0 200>))
   /SetVariables(Particles.FadeOffAlpha:NiagaraFloat=&ClampFloat(Float=&Multiply_Float(A=$Emitter.Age, B=1))) {isSetParameters=true, module=SetVariables_#1}
  ]
  particleUpdateScript [
   &SolveForcesAndVelocity(Clamp Velocity=false, Limit Acceleration=false)
   &FN_UpdateRibbon(
    Enable Skeletal Position=true
    Initial Color:LinearColor=$Particles.Initial.Color
    Initial RibbonWidth:NiagaraFloat=$Particles.Initial.RibbonWidth
   )
   &Material(Base Texture Time:NiagaraFloat=&Multiply_Float(A=$Emitter.Age, B=-1))
   &DynamicMaterialParameters(
    Index 0 Param 1=1
    Index 0 Param 2=1
    Index 0 Param 3=0.1
    Index 0 Param 4=0.2
    Index 1 Param 1=&Multiply_Float(A=-1.5, B=$Emitter.Age)
    Index 1 Param 2=0.15
    Index 1 Param 3=1.3
    Index 1 Param 4=1.8
    Index 2 Param 1=0.2
    Index 2 Param 2=0.2
    Index 2 Param 3=0.2
    Index 3 Param 1=1
    Index 3 Param 2=1
    Index 3 Param 3=1
    Index 3 Param 4=1
    Module.Param0WriteEnabled=true
    Module.Param0WriteEnabled1=true
    Module.Param0WriteEnabled2=true
    Module.Param0WriteEnabled3=true
    Module.Param1WriteEnabled=true
    Module.Param1WriteEnabled1=true
    Module.Param1WriteEnabled2=true
    Module.Param1WriteEnabled3=true
    Module.Param2WriteEnabled=true
    Module.Param2WriteEnabled1=true
    Module.Param2WriteEnabled2=true
    Module.Param2WriteEnabled3=true
    Module.Param3WriteEnabled=true
    Module.Param3WriteEnabled1=true
    Module.Param3WriteEnabled2=true
    Module.Param3WriteEnabled3=true
    Write Parameter Index 1=true
    Write Parameter Index 2=true
    Write Parameter Index 3=true
   )
   &ScaleRibbonWidth(Ribbon Width Scale=&FloatFromCurve(FloatCurve=DataInterface() {propertyValues {Curve {keys [
         <at 0.0039 1.3394 RCIM_Cubic>
         {arriveTangent=-0.4407, interpMode=RCIM_Cubic, leaveTangent=-0.4407, time=0.1718, value=1.5109}
         <at 1.0045 0.8984>
        ]}}}))
   &Color(
    Color=&ColorFromCurve(ColorCurve=DataInterface() {propertyValues {
       AlphaCurve {keys [<at -0.0023 1 RCIM_Cubic>,<at 0.9977 1 RCIM_Cubic>]}
       BlueCurve {keys [<at -0.0023 0.0671 RCIM_Cubic>,<at 0.6301 0.2435>,<at 1.0058 0.2135 RCIM_Cubic>]}
       GreenCurve {keys [<at -0.0023 1 RCIM_Cubic>,<at 0.6301 0.2865>,<at 1.0058 0 RCIM_Cubic>]}
       RedCurve {keys [<at -0.0023 0.2815 RCIM_Cubic>,<at 0.6301 0.0054>,<at 1.0058 0.014 RCIM_Cubic>]}
      }})
    Scale Alpha=&Multiply_Float(
     A=&FloatFromCurve(
      FloatCurve=DataInterface() {propertyValues {Curve {keys [{arriveTangent=-2.0086, interpMode=RCIM_Cubic, leaveTangent=-2.0086, tangentMode=RCTM_User, time=-0.0046, value=1.0113}, <at 1.0011 0.002>]}}}
      Scale Curve=&FloatFromCurve(FloatCurve=DataInterface() {propertyValues {Curve {keys [<at 0.3 0>,<at 0.9974 0.1>]}}}, CurveIndex=$Emitter.Age)
     )
     B=&FNTimeOfDayScalar(Night Scale=1, Morning Scale=0.3, Evening Scale=0.3)
    )
    Scale Color=<4 4 1>
   )
   &KillParticles(Kill Particles=&SetBoolByFloatComparison(A=$Particles.Age, B=$Particles.Lifetime, Comparison Type=::NewEnumerator2))
  ]
  rendererClasses [&NiagaraRibbonRendererProperties]
  renderers [{rendererClass=&NiagaraRibbonRendererProperties}]
 }
 Contrail_Red {
  emitterUpdateScript [
   &EmitterState()
   &FN_SpawnRibbon_v2(Number of Ribbons=2, Spawn Mode=::NewEnumerator0, SpawnRate=5)
  ]
  particleSpawnScript [
   &FN_InitializeRibbon_v5(Enable Skeletal Position=true, Lifetime=1, Ribbon Width=35, SkeletalMesh:NiagaraDataInterfaceSkeletalMesh=$User.SkeletalMesh)
   &AddVelocity(Velocity=&FN_FrontEndVector3Modifier(Gameplay=<0 0 500>, Front End=<0 0 200>))
   /SetVariables(Particles.FadeOffAlpha:NiagaraFloat=&ClampFloat(Float=&Multiply_Float(A=$Emitter.Age, B=1))) {isSetParameters=true, module=SetVariables_#1}
  ]
  particleUpdateScript [
   &SolveForcesAndVelocity(Clamp Velocity=false, Limit Acceleration=false)
   &FN_UpdateRibbon(
    Enable Skeletal Position=true
    Initial Color:LinearColor=$Particles.Initial.Color
    Initial RibbonWidth:NiagaraFloat=$Particles.Initial.RibbonWidth
   )
   &Material(Base Texture Time:NiagaraFloat=&Multiply_Float(A=$Emitter.Age, B=-1))
   &DynamicMaterialParameters(
    Index 0 Param 1=1
    Index 0 Param 2=1
    Index 0 Param 3=0.1
    Index 0 Param 4=0.2
    Index 1 Param 1=&Multiply_Float(A=-1.3, B=$Emitter.Age)
    Index 1 Param 2=0.15
    Index 1 Param 3=1.3
    Index 1 Param 4=1.8
    Index 2 Param 1=0.2
    Index 2 Param 2=0.2
    Index 2 Param 3=0.2
    Index 3 Param 1=1
    Index 3 Param 2=1
    Index 3 Param 3=1
    Index 3 Param 4=1
    Module.Param0WriteEnabled=true
    Module.Param0WriteEnabled1=true
    Module.Param0WriteEnabled2=true
    Module.Param0WriteEnabled3=true
    Module.Param1WriteEnabled=true
    Module.Param1WriteEnabled1=true
    Module.Param1WriteEnabled2=true
    Module.Param1WriteEnabled3=true
    Module.Param2WriteEnabled=true
    Module.Param2WriteEnabled1=true
    Module.Param2WriteEnabled2=true
    Module.Param2WriteEnabled3=true
    Module.Param3WriteEnabled=true
    Module.Param3WriteEnabled1=true
    Module.Param3WriteEnabled2=true
    Module.Param3WriteEnabled3=true
    Write Parameter Index 1=true
    Write Parameter Index 2=true
    Write Parameter Index 3=true
   )
   &ScaleRibbonWidth(Ribbon Width Scale=&FloatFromCurve(FloatCurve=DataInterface() {propertyValues {Curve {keys [
         <at 0.0039 1.3394 RCIM_Cubic>
         {arriveTangent=-0.4407, interpMode=RCIM_Cubic, leaveTangent=-0.4407, time=0.1718, value=1.5109}
         <at 1.0045 0.8984>
        ]}}}))
   &CameraOffset(Camera Offset Amount=10)
   &Color(
    Color=&ColorFromCurve(ColorCurve=DataInterface() {propertyValues {
       AlphaCurve {keys [<at -0.0023 1 RCIM_Cubic>,<at 0.9977 1 RCIM_Cubic>]}
       BlueCurve {keys [<at -0.0023 0 RCIM_Cubic>,<at 0.6301 0>]}
       GreenCurve {keys [<at -0.0023 0.2201 RCIM_Cubic>,<at 0.6301 0.0162>]}
       RedCurve {keys [<at -0.0023 1 RCIM_Cubic>,<at 0.6301 0.2865>]}
      }})
    Scale Alpha=&Multiply_Float(
     A=&FloatFromCurve(
      FloatCurve=DataInterface() {propertyValues {Curve {keys [{arriveTangent=-2.0086, interpMode=RCIM_Cubic, leaveTangent=-2.0086, tangentMode=RCTM_User, time=-0.0046, value=1.0113}, <at 1.0011 0.002>]}}}
      Scale Curve=&FloatFromCurve(FloatCurve=DataInterface() {propertyValues {Curve {keys [<at 0.3 0>,<at 1 0.1>]}}}, CurveIndex=$Emitter.Age)
     )
     B=&FNTimeOfDayScalar(Night Scale=1, Morning Scale=0.3, Evening Scale=0.3)
    )
    Scale Color=<4 4 1>
   )
   &KillParticles(Kill Particles=&SetBoolByFloatComparison(A=$Particles.Age, B=$Particles.Lifetime, Comparison Type=::NewEnumerator2))
  ]
  rendererClasses [&NiagaraRibbonRendererProperties]
  renderers [{rendererClass=&NiagaraRibbonRendererProperties}]
 }
 FN_Contrails_AttachedParticle {
  emitterSpawnScript [&FN_PreserveSkydiveUserVariables() {module=PreserveUserVariables}]
  emitterUpdateScript [
   &EmitterState()
   &SpawnRate(SpawnRate=4, Use Spawn Probability=false)
  ]
  particleSpawnScript [
   &FN_SetSpawnGroup_v2(Number of Spawn Groups=2)
   /Niagara/Modules/Spawn/Initialization/InitializeParticle(
    Lifetime=0.6
    Module.Write Color=true
    Module.Write Lifetime=true
    Module.Write Mass=true
    Module.Write Material Random=true
    Module.Write Position=false
    Module.Write Scale=false
    Module.Write SpriteRotation=false
    Module.Write SpriteSize=true
    Position=&SimulationPosition()
    Sprite Size=&Vector2DFromFloat(Value=&UniformRangedFloat(Minimum=40, Maximum=80))
   )
   &CameraOffset(Camera Offset Amount=&UniformRangedFloat(Minimum=10, Maximum=30))
  ]
  particleUpdateScript [
   /SetVariables(Particles.Color=&MakeLinearColorFromVectorAndFloat(
     "Vector (RGB)"=&MakeVectorFromLinearColorRGB(Color=&ColorFromCurve(ColorCurve=DataInterface() {propertyValues {
         AlphaCurve {keys [<at -0.0023 1 RCIM_Cubic>,<at 0.9977 1 RCIM_Cubic>]}
         BlueCurve {keys [<at -0.0036 0.0791 RCIM_Cubic>,<at 0.341 0.1071>,<at 0.5867 0.5054>,<at 0.8035 0 RCIM_Cubic>]}
         GreenCurve {keys [<at -0.0036 0.1156 RCIM_Cubic>,<at 0.341 1>,<at 0.5867 1>,<at 0.8035 0.4573 RCIM_Cubic>]}
         RedCurve {keys [<at -0.0036 1 RCIM_Cubic>,<at 0.341 0.7069>,<at 0.5867 0>,<at 0.8035 1 RCIM_Cubic>]}
        }}))
     "Float (Alpha)"=&Multiply_Float(
      A=&FloatFromCurve(FloatCurve=DataInterface() {propertyValues {Curve {keys [<at 0 0 RCIM_Cubic>,<at 0.5 0.3 RCIM_Cubic>,<at 1 0>]}}})
      B=&FNTimeOfDayScalar(Night Scale=0.5, Morning Scale=0.3, Evening Scale=0.3)
     )
    )) {isSetParameters=true}
   /SetVariables(Particles.Light_Color:LinearColor=&MakeLinearColorFromVectorAndFloat(
     "Vector (RGB)"=&MakeVectorFromLinearColorRGB(Color=&MultiplyLinearColor(
       Linear Color=&ColorFromCurve(ColorCurve=DataInterface() {propertyValues {
          AlphaCurve {keys [<at -0.0023 1 RCIM_Cubic>,<at 0.9977 1 RCIM_Cubic>]}
          BlueCurve {keys [<at -0.0036 0.0791 RCIM_Cubic>,<at 0.341 0.1071>,<at 0.5867 0.5054>,<at 0.8035 0 RCIM_Cubic>]}
          GreenCurve {keys [<at -0.0036 0.1156 RCIM_Cubic>,<at 0.341 1>,<at 0.5867 1>,<at 0.8035 0.4573 RCIM_Cubic>]}
          RedCurve {keys [<at -0.0036 1 RCIM_Cubic>,<at 0.341 0.7069>,<at 0.5867 0>,<at 0.8035 1 RCIM_Cubic>]}
         }})
       Scale Factor=&MakeVector4_FromFloat(Float=30)
      ))
     "Float (Alpha)"=&Multiply_Float(
      A=1
      B=&FloatFromCurve(FloatCurve=DataInterface() {propertyValues {Curve {keys [<at 0 0 RCIM_Cubic>,<at 0.498 1 RCIM_Cubic>,<at 1 0>]}}})
     )
    )) {isSetParameters=true}
   /SetVariables(Particles.Light_falloff=2) {isSetParameters=true}
   &ParticleState()
   &SkeletalMeshLocation(
    Bone / Socket Index=$Particles.SpawnGroup
    Bone Sampling Mode=::NewEnumerator4
    Inherit Velocity=false
    Interpolate Mesh Sampling=false
    MaskSpawnGroup=false
    Sampled Position Offset Coordinate Space:ENiagaraMeshTransforms=::NewEnumerator3
    Skeletal Mesh=$User.SkeletalMesh
   )
   &SolveForcesAndVelocity(Clamp Velocity=false, Limit Acceleration=false)
   &ScaleMeshSize(Scale Factor=&Multiply_Vector(
     A=<1 1 1>
     B=&VectorFromFloat(Value=&FloatFromCurve(FloatCurve=DataInterface() {propertyValues {Curve {keys [<at 0 0>,<at 1 1.5>]}}}))
    ))
  ]
  rendererClasses [&NiagaraSpriteRendererProperties, &NiagaraLightRendererProperties]
  renderers [
   {rendererClass=&NiagaraSpriteRendererProperties}
   {rendererClass=&NiagaraLightRendererProperties, rendererIndex=1}
  ]
 }
 FN_Contrails_CookieMan {
  emitterUpdateScript [
   &EmitterState()
   &SpawnRate(
    SpawnRate=&Multiply_Float(
     A=8
     B=&FloatFromCurve(
      FloatCurve=DataInterface() {propertyValues {Curve {keys [<at 0 0.7>,<at 1 1.5>]}}}
      CurveIndex=&Divide_Float(A=&VectorLength(Vector=$Engine.Owner.Velocity), B=9000)
     )
    )
    Use Spawn Probability=false
   )
  ]
  particleSpawnScript [
   &FN_SetSpawnGroup_v2(Number of Spawn Groups=2)
   &InitializeParticle(
    Color=<rgba 1 0.8417 0.7639 1>
    Lifetime=&UniformRangedFloat(Minimum=1.1, Maximum=1.2)
    Mesh Scale=&Multiply_Vector(
     A=<1 1 1>
     B=&VectorFromFloat(Value=&UniformRangedFloat(Minimum=0.1, Maximum=0.25))
    )
    Module.Write Color=true
    Module.Write Lifetime=true
    Module.Write Mass=true
    Module.Write Material Random=true
    Module.Write Position=false
    Module.Write Scale=true
    Module.Write SpriteRotation=false
    Module.Write SpriteSize=false
    Position=&SimulationPosition()
   )
   &CylinderLocation(
    Cylinder Height=1
    Lathe Profile=DataInterface() {propertyValues {Curve {keys []}}}
    MaskSpawnGroup=false
    Module.Enable Lathe Profile=false
    Module.Override Randomness=false
    Module.Surface Only=false
   )
   &AddVelocity(Velocity=&MakeVector(
     X=0
     Y=0
     Z=&Multiply_Float(
      A=200
      B=&FloatFromCurve(
       FloatCurve=DataInterface() {propertyValues {Curve {keys [<at 0 0.9>,<at 1 2>]}}}
       CurveIndex=&Divide_Float(A=&VectorLength(Vector=$Engine.Owner.Velocity), B=9000)
      )
     )
    ))
   /Niagara/Modules/Spawn/Orientation/V2/InitialMeshOrientation(
    Orientation Vector=&UniformRangedVector(Minimum=<-0.5 -0.5 -0.5>, Maximum=<0.5 0.5 0.5>)
    Use Orientation Vector=true
    Use Rotation Vector=false
   )
   /SetVariables(Particles.RotationOrientation:NiagaraFloat=&UniformRangedFloat(Minimum=-1, Maximum=1)) {isSetParameters=true, module=SetVariables_#2}
   /SetVariables(Particles.RotationSpeed:Vector3f=&UniformRangedVector(Minimum=<-0.5 -0.5 -0.5>, Maximum=<0.5 0.5 0.5>)) {isSetParameters=true, module=SetVariables_#3}
   &SkeletalMeshLocation(
    Bone / Socket Index=$Particles.SpawnGroup
    Bone Sampling Mode=::NewEnumerator4
    Inherit Velocity=false
    MaskSpawnGroup=false
    Skeletal Mesh=$User.SkeletalMesh
   )
  ]
  particleUpdateScript [
   &ParticleState()
   /Niagara/Modules/Update/Orientation/V2/UpdateMeshOrientation(Rotation Rate:NiagaraFloat=$Particles.RotationOrientation, Rotation Vector:Vector3f=$Particles.RotationSpeed)
   &Drag(Drag=0.3, Use Linear Drag=true, Use Rotational Drag=false)
   /Niagara/Modules/Update/Forces/CurlNoiseForce(Noise Frequency=0.1, Noise Strength=600, Pan Noise=false, Pan Noise Field=<1 1 0>)
   &DynamicMaterialParameters(
    Index 0 Param 1=&FloatFromCurve(FloatCurve=DataInterface() {propertyValues {Curve {keys [<at -0.0046 1.7>,<at 1.0011 0.002>]}}})
    Module.Param0WriteEnabled=true
    Module.Param1WriteEnabled=true
    Module.Param2WriteEnabled=true
    Module.Param3WriteEnabled=true
   )
   &ScaleMeshSize(Scale Factor=&Multiply_Vector(
     A=<1 1 1>
     B=&VectorFromFloat(Value=&FloatFromCurve(FloatCurve=DataInterface() {propertyValues {Curve {keys [<at -0.0046 1>,<at 0.2917 0.8093>,<at 0.366 1.2044>,<at 0.4265 0.7466>,<at 0.4835 0.7129>,<at 0.531 0.9096>,<at 0.5888 0.6713>,<at 0.7031 0.6175>,<at 0.784 0.5855>,<at 1.0011 0.5>]}}}))
    ))
   &InheritVelocity(Module.Limit Speed=true, Velocity Limit=1000, Velocity Scale=<0 0 -0.01>) {enabled=false}
   &SolveForcesAndVelocity(Clamp Velocity=false, Limit Acceleration=false)
   &GenerateLocationEvent(Event Send Rate:NiagaraFloat=&UniformRangedFloat(Minimum=1, Maximum=3), Module.Use Event Send Rate=true)
  ]
  rendererClasses [&NiagaraMeshRendererProperties]
  renderers [{rendererClass=&NiagaraMeshRendererProperties, rendererIndex=0}]
 }
 FN_Contrails_CookieMan_Fragment {
  emitterUpdateScript [&EmitterState()]
  particleSpawnScript [
   &FN_SetSpawnGroup_v2(Number of Spawn Groups=2)
   &InitializeParticle(
    Color=<rgba 2 2 2 1>
    Mesh Scale=&Multiply_Vector(
     A=<1 1 1>
     B=&VectorFromFloat(Value=&UniformRangedFloat(Minimum=0.03, Maximum=0.07))
    )
    Module.Write Color=true
    Module.Write Lifetime=true
    Module.Write Mass=true
    Module.Write Material Random=true
    Module.Write Position=false
    Module.Write Scale=true
    Module.Write SpriteRotation=false
    Module.Write SpriteSize=false
    Position=&SimulationPosition()
   )
   /Niagara/Modules/Spawn/Location/SphereLocation(MaskSpawnGroup=false, Module.Override Randomness=false, Sphere Radius=30, Surface Only=false)
   &AddVelocity(Velocity=&UniformRangedVector(Minimum=<-50 -50 0>, Maximum=<50 50 100>)) {module=AddVelocity001}
   /SetVariables(Particles.RotationOrientation:NiagaraFloat=&UniformRangedFloat(Minimum=-1, Maximum=1)) {isSetParameters=true, module=SetVariables_#4}
   &InheritVelocity(Module.Limit Speed=false, Velocity Scale=<0 0 -0.03>)
   /SetVariables(Particles.RotationSpeed:Vector3f=&UniformRangedVector(Minimum=<-1 -1 -1>, Maximum=<1 1 1>)) {isSetParameters=true, module=SetVariables_#5}
   &InitialMeshOrientation(
    Orientation Vector=&UniformRangedVector(Minimum=<-1 -1 -1>, Maximum=<1 1 1>)
    Use Orientation Vector=true
    Use Rotation Vector=false
   )
  ]
  particleUpdateScript [
   &ParticleState()
   &SolveForcesAndVelocity(Clamp Velocity=false, Limit Acceleration=false)
   &ScaleMeshSize(Scale Factor=&Multiply_Vector(
     A=<1 1 1>
     B=&VectorFromFloat(Value=&FloatFromCurve(FloatCurve=DataInterface() {propertyValues {Curve {keys [<at -0.0046 0>,<at 0.0542 0.0143>,<at 0.0756 0.9337>,<at 1.0011 0>]}}}))
    ))
  ]
  rendererClasses [&NiagaraMeshRendererProperties]
  renderers [{rendererClass=&NiagaraMeshRendererProperties, rendererIndex=0}]
 }
 FN_Contrails_CookieTree {
  emitterSpawnScript [&FN_PreserveSkydiveUserVariables() {module=PreserveUserVariables}]
  emitterUpdateScript [
   &EmitterState()
   &SpawnRate(
    SpawnRate=&Multiply_Float(
     A=8
     B=&FloatFromCurve(
      FloatCurve=DataInterface() {propertyValues {Curve {keys [<at 0 0.7>,<at 1 1.5>]}}}
      CurveIndex=&Divide_Float(A=&VectorLength(Vector=$Engine.Owner.Velocity), B=9000)
     )
    )
    Use Spawn Probability=false
   ) {module=SpawnRate001}
  ]
  particleSpawnScript [
   &FN_SetSpawnGroup_v2(Number of Spawn Groups=2)
   &InitializeParticle(
    Color=<rgba 0.3554 0.229 0.6233 1>
    Lifetime=&UniformRangedFloat(Minimum=1.1, Maximum=1.2)
    Mesh Scale=&Multiply_Vector(
     A=<1 1 1>
     B=&VectorFromFloat(Value=&UniformRangedFloat(Minimum=0.1, Maximum=0.5))
    )
    Module.Write Color=true
    Module.Write Lifetime=true
    Module.Write Mass=true
    Module.Write Material Random=true
    Module.Write Position=false
    Module.Write Scale=true
    Module.Write SpriteRotation=false
    Module.Write SpriteSize=false
    Position=&SimulationPosition()
   )
   &CylinderLocation(
    Cylinder Height=1
    Lathe Profile=DataInterface() {propertyValues {Curve {keys []}}}
    MaskSpawnGroup=false
    Module.Enable Lathe Profile=false
    Module.Override Randomness=false
    Module.Surface Only=false
   )
   &AddVelocity(Velocity=&UniformRangedVector(
     Minimum=&MakeVector(
      X=0
      Y=0
      Z=&Multiply_Float(
       A=150
       B=&FloatFromCurve(
        FloatCurve=DataInterface() {propertyValues {Curve {keys [<at 0 0.9>,<at 1 2>]}}}
        CurveIndex=&Divide_Float(A=&VectorLength(Vector=$Engine.Owner.Velocity), B=9000)
       )
      )
     )
     Maximum=&MakeVector(
      X=0
      Y=0
      Z=&Multiply_Float(
       A=200
       B=&FloatFromCurve(
        FloatCurve=DataInterface() {propertyValues {Curve {keys [<at 0 0.9>,<at 1 2>]}}}
        CurveIndex=&Divide_Float(A=&VectorLength(Vector=$Engine.Owner.Velocity), B=9000)
       )
      )
     )
    ))
   &InitialMeshOrientation(
    Orientation Vector=&UniformRangedVector(Minimum=<-0.5 -0.5 -0.5>, Maximum=<0.5 0.5 0.5>)
    Use Orientation Vector=true
    Use Rotation Vector=false
   )
   /SetVariables(Particles.RotationOrientation:NiagaraFloat=&UniformRangedFloat(Minimum=-1, Maximum=1)) {isSetParameters=true, module=SetVariables_#2}
   /SetVariables(Particles.RotationSpeed:Vector3f=&UniformRangedVector(Minimum=<-0.5 -0.5 -0.5>, Maximum=<0.5 0.5 0.5>)) {isSetParameters=true, module=SetVariables_#3}
   &SkeletalMeshLocation(
    Bone / Socket Index=$Particles.SpawnGroup
    Bone Sampling Mode=::NewEnumerator4
    Inherit Velocity=false
    MaskSpawnGroup=false
    Skeletal Mesh=$User.SkeletalMesh
   )
  ]
  particleUpdateScript [
   &ParticleState()
   &UpdateMeshOrientation(Rotation Rate:NiagaraFloat=$Particles.RotationOrientation, Rotation Vector:Vector3f=$Particles.RotationSpeed)
   &Drag(Drag=0.3, Use Linear Drag=true, Use Rotational Drag=false)
   &CurlNoiseForce(
    Noise Frequency:NiagaraFloat=&UniformRangedFloat(Minimum=0.1, Maximum=1)
    Noise Strength=&UniformRangedFloat(Minimum=300, Maximum=600)
    Pan Noise=false
    Pan Noise Field=<1 1 0>
   )
   &DynamicMaterialParameters(
    Index 0 Param 1=&FloatFromCurve(FloatCurve=DataInterface() {propertyValues {Curve {keys [<at -0.0046 2>,<at 1.0011 0.002>]}}})
    Module.Param0WriteEnabled=true
    Module.Param1WriteEnabled=true
    Module.Param2WriteEnabled=true
    Module.Param3WriteEnabled=true
   )
   &ScaleMeshSize(Scale Factor=&Multiply_Vector(
     A=<1 1 1>
     B=&VectorFromFloat(Value=&FloatFromCurve(FloatCurve=DataInterface() {propertyValues {Curve {keys [
           <at -0.0046 1>
           <at 0.3771 0.7748>
           {arriveTangent=-0.3128, leaveTangent=-0.3128, time=0.4459, value=1.2326}
           <at 0.5037 0.6933>
           <at 0.5587 0.6619>
           <at 0.611 0.8501>
           <at 0.6536 0.618>
           <at 0.7155 0.6055>
           <at 0.8338 0.5741>
           <at 1.0011 0.5>
          ]}}}))
    ))
   &SolveForcesAndVelocity(Clamp Velocity=false, Limit Acceleration=false)
   &GenerateLocationEvent(Event Send Rate:NiagaraFloat=&UniformRangedFloat(Minimum=1, Maximum=3), Module.Use Event Send Rate=true)
  ]
  rendererClasses [&NiagaraMeshRendererProperties]
  renderers [{rendererClass=&NiagaraMeshRendererProperties, rendererIndex=0}]
 }
 FN_Contrails_CookieTree_Fragment {
  emitterUpdateScript [&EmitterState()]
  particleSpawnScript [
   &FN_SetSpawnGroup_v2(Number of Spawn Groups=2)
   &InitializeParticle(
    Color=<rgba 0.0783 0.6233 0.0188 1>
    Mesh Scale=&Multiply_Vector(
     A=<1 1 1>
     B=&VectorFromFloat(Value=&UniformRangedFloat(Minimum=0.03, Maximum=0.07))
    )
    Module.Write Color=true
    Module.Write Lifetime=true
    Module.Write Mass=true
    Module.Write Material Random=true
    Module.Write Position=false
    Module.Write Scale=true
    Module.Write SpriteRotation=false
    Module.Write SpriteSize=false
    Position=&SimulationPosition()
   )
   &SphereLocation(MaskSpawnGroup=false, Module.Override Randomness=false, Sphere Radius=30, Surface Only=false)
   &AddVelocity(Velocity=&UniformRangedVector(Minimum=<-50 -50 0>, Maximum=<50 50 100>)) {module=AddVelocity001}
   /SetVariables(Particles.RotationOrientation:NiagaraFloat=&UniformRangedFloat(Minimum=-1, Maximum=1)) {isSetParameters=true, module=SetVariables_#4}
   &InheritVelocity(Module.Limit Speed=false, Velocity Scale=<0 0 -0.03>)
   /SetVariables(Particles.RotationSpeed:Vector3f=&UniformRangedVector(Minimum=<-1 -1 -1>, Maximum=<1 1 1>)) {isSetParameters=true, module=SetVariables_#5}
   &InitialMeshOrientation(
    Orientation Vector=&UniformRangedVector(Minimum=<-1 -1 -1>, Maximum=<1 1 1>)
    Use Orientation Vector=true
    Use Rotation Vector=false
   )
  ]
  particleUpdateScript [
   &ParticleState()
   &UpdateMeshOrientation(Rotation Rate:NiagaraFloat=$Particles.RotationOrientation, Rotation Vector:Vector3f=$Particles.RotationSpeed)
   &SolveForcesAndVelocity(Clamp Velocity=false, Limit Acceleration=false)
   &ScaleMeshSize(Scale Factor=&Multiply_Vector(
     A=<1 1 1>
     B=&VectorFromFloat(Value=&FloatFromCurve(FloatCurve=DataInterface() {propertyValues {Curve {keys [<at -0.0046 0>,<at 0.0542 0.0143>,<at 0.0756 0.9337>,<at 1.0011 0>]}}}))
    )) {module=ScaleMeshSize001}
  ]
  rendererClasses [&NiagaraMeshRendererProperties]
  renderers [{rendererClass=&NiagaraMeshRendererProperties, rendererIndex=0}]
 }
 FN_Contrails_Star {
  emitterUpdateScript [
   &EmitterState()
   &SpawnRate(SpawnRate=40, Use Spawn Probability=false)
  ]
  particleSpawnScript [
   &FN_SetSpawnGroup_v2(Number of Spawn Groups=2)
   &InitializeParticle(
    Color=<rgba 1 0.1583 0.0084 1>
    Lifetime=&UniformRangedFloat(Minimum=0.3, Maximum=0.5)
    Mesh Scale=&Multiply_Vector(
     A=<1 1 1>
     B=&VectorFromFloat(Value=&UniformRangedFloat(Minimum=0.2, Maximum=1))
    )
    Module.Write Color=true
    Module.Write Lifetime=true
    Module.Write Mass=true
    Module.Write Material Random=true
    Module.Write Position=false
    Module.Write Scale=true
    Module.Write SpriteRotation=false
    Module.Write SpriteSize=false
    Position=&SimulationPosition()
   )
   &CylinderLocation(
    Cylinder Height=1
    Lathe Profile=DataInterface() {propertyValues {Curve {keys []}}}
    MaskSpawnGroup=false
    Module.Enable Lathe Profile=false
    Module.Override Randomness=false
    Module.Surface Only=false
   )
   &SkeletalMeshLocation(
    Bone / Socket Index=$Particles.SpawnGroup
    Bone Sampling Mode=::NewEnumerator4
    Inherit Velocity=false
    MaskSpawnGroup=false
    Skeletal Mesh=$User.SkeletalMesh
   )
   &AddVelocity(Velocity=&UniformRangedVector(Minimum=<0 0 200>, Maximum=<1 1 80>))
   &InitialMeshOrientation(
    Orientation Vector=&UniformRangedVector(Minimum=<-1 -1 -1>, Maximum=<1 1 1>)
    Use Orientation Vector=true
    Use Rotation Vector=false
   )
   /SetVariables(Particles.RotationOrientation:NiagaraFloat=&UniformRangedFloat(Minimum=-1, Maximum=1)) {isSetParameters=true, module=SetVariables_#2}
   /SetVariables(Particles.RotationSpeed:Vector3f=&UniformRangedVector(Minimum=<-1 -1 -1>, Maximum=<1 1 1>)) {isSetParameters=true, module=SetVariables_#3}
   &SphereLocation(MaskSpawnGroup=false, Module.Override Randomness=false, Sphere Radius=20, Surface Only=false)
  ]
  particleUpdateScript [
   &ParticleState()
   &Drag(Drag=0.5, Use Linear Drag=true, Use Rotational Drag=false)
   &CurlNoiseForce(
    Noise Frequency:NiagaraFloat=&UniformRangedFloat(Minimum=0.1, Maximum=1)
    Noise Strength=&UniformRangedFloat(Minimum=300, Maximum=600)
    Pan Noise=false
    Pan Noise Field=<1 1 0>
   )
   &SolveForcesAndVelocity(Clamp Velocity=false, Limit Acceleration=false)
   &ScaleSpriteSize(Scale Factor=&Multiply_Vector2(
     A=<1 1>
     B=&Vector2DFromFloat(Value=&FloatFromCurve(FloatCurve=DataInterface() {propertyValues {Curve {keys [<at -0.0046 3>,<at 1.0011 0.002>]}}}))
    ))
   /SetVariables(Particles.Color=&MakeLinearColorFromVectorAndFloat(
     "Vector (RGB)"=&MakeVectorFromLinearColorRGB(Color=&ColorFromCurve(ColorCurve=DataInterface() {propertyValues {
         AlphaCurve {keys [<at -0.0023 1 RCIM_Cubic>,<at 0.9977 1 RCIM_Cubic>]}
         BlueCurve {keys [<at 0.1996 0.0791 RCIM_Cubic>,<at 0.4082 0.1071>,<at 0.5661 0.5054>,<at 0.7637 0>]}
         GreenCurve {keys [<at 0.1996 0.1156 RCIM_Cubic>,<at 0.4082 1>,<at 0.5661 1>,<at 0.7637 0.4573>]}
         RedCurve {keys [<at 0.1996 1 RCIM_Cubic>,<at 0.4082 0.7069>,<at 0.5661 0>,<at 0.7637 1>]}
        }}))
     "Float (Alpha)"=&FloatFromCurve(FloatCurve=DataInterface() {propertyValues {Curve {keys [<at -0.0046 1.0113>,<at 0.1223 0.0095>,<at 0.2498 0.8891>,<at 0.4241 -0.0041>,<at 0.5572 0.9199>,<at 0.7294 0.0749>,<at 0.8767 0.9384>,<at 1.0011 0.002>]}}})
    )) {isSetParameters=true}
  ]
  rendererClasses [&NiagaraSpriteRendererProperties]
  renderers [{rendererClass=&NiagaraSpriteRendererProperties}]
 }
}
summary {
 systemName=NS_Contrail_HolidayCookiesV3
 emitters [
  {emitterName=FN_Contrails_CookieMan, rendererClasses [&NiagaraMeshRendererProperties]}
  {emitterName=FN_Contrails_Star, rendererClasses [&NiagaraSpriteRendererProperties]}
  {emitterName=FN_Contrails_CookieTree, rendererClasses [&NiagaraMeshRendererProperties]}
  {emitterName=FN_Contrails_CookieMan_Fragment, rendererClasses [&NiagaraMeshRendererProperties]}
  {emitterName=FN_Contrails_CookieTree_Fragment, rendererClasses [&NiagaraMeshRendererProperties]}
  {
   emitterName=FN_Contrails_AttachedParticle
   rendererClasses [&NiagaraSpriteRendererProperties, &NiagaraLightRendererProperties]
  }
  {emitterName=Contrail_Green, rendererClasses [&NiagaraRibbonRendererProperties]}
  {emitterName=Contrail_Red, rendererClasses [&NiagaraRibbonRendererProperties]}
 ]
 userVariables [
  User.bIsFrontEnd {type=NiagaraBool}
  User.bIsFrontEndPreview {type=NiagaraBool}
  User.FEVec {type=Vector3f, description="", defaultValue {struct=Vector3f, value=<0 0 1000>}}
  User.PlayerSpeed {type=NiagaraFloat, defaultValue=0, description=""}
  User.RibbonIDOffset {type=NiagaraInt32}
  User.SkeletalMesh {
   type=NiagaraDataInterfaceSkeletalMesh
   defaultValue {
    struct=NiagaraExt_VariableValue_DataInterface
    value {dataInterface=&self:NiagaraDataInterfaceSkeletalMesh_0}
   }
  }
  User.SpawnRate {type=NiagaraFloat, defaultValue=360, description=""}
 ]
}
